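# Niagara system for a brownish ground dust puff in the Pizza Party player emote.
# Two burst sprite emitters spawn on flat rings: Smoke_Ground shoots outward under decaying drag with viewpoint-based rotation, while Directional_Smoke_Ground001 adds shorter-lived, tall, narrow streaks.
# Both grow over life, pick random flipbook frames, animate material parameters and fade alpha scaled by time of day.
# A system module sets a foot-dust color array index, though both emitters tint through their own color curves.
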NiagaraSystem /Game/Animation/Game/MainPlayer/Emotes/PizzaParty/FX/NS_Emote_PizzaParty_DustPuff
dependencies {
 usedDataInterfaces [&NiagaraDataInterfaceCamera, &NiagaraDataInterfaceCurve, &NiagaraDataInterfaceColorCurve]
 usedDynamicInputs [&Vector2DFromFloat]
 usedModules [
  &SystemState
  &self:SetFootDustFXColorArrayIndex
 ]
}
emitters {
 Directional_Smoke_Ground001 {
  emitterUpdateScript [
   &EmitterState()
   &SpawnBurst_Instantaneous(Spawn Count=6, Use Spawn Probability=false)
  ]
  particleSpawnScript [
   &InitializeParticle(
    Color=<rgba 1 1 1 0.2>
    Lifetime=&RandomRangeFloat(Minimum=0.22, Maximum=0.25)
    Sprite Rotation Mode=::NewEnumerator2
    Sprite Size Max=<15 120>
    Sprite Size Min=<12 100>
    Sprite Size Mode=::NewEnumerator2
    UsePositionOffset=false
    Write Material Random=true
   )
   &ShapeLocation(
    Cylinder Distribution=::NewEnumerator2
    Cylinder Height=0
    Cylinder Radius=15
    MaskSpawnGroup=false
    Rotation Coordinate Space=::World
    Shape Primitive=::NewEnumerator1
    U Position=&ReturnNormalizedExecIndex(Normalized Index Includes One=false)
    UseOwnerScale=true
   )
   &AddVelocity(
    Rotation Coordinate Space=::World
    Velocity=&Multiply_VectorByFloat(Vector=$Particles.ShapeLocation.ShapeNormal, Float=&RandomRangeFloat(Minimum=0.8, Maximum=1))
   )
   &AddVelocity(Rotation Coordinate Space=::World, Velocity=&RandomRangeVector(Minimum=<0 0 0.7>, Maximum=<0 0 0.8>)) {module=AddVelocity001}
   &ApplyInitialForces()
   /SetVariables(Particles.FirstFlipBookFrame:NiagaraInt32=&RandomRangeInt(Minimum=0, Maximum=64)) {isSetParameters=true, module=SetVariables_#1}
   /SetVariables(
    Particles.Stored Particle Size:Vector3f=$Particles.Scale
    Particles.Stored Sprite Size:Vector2f=&Multiply_Vector2D_ByFloat(Float=1, Vector2D=$Particles.SpriteSize)
   ) {isSetParameters=true, module=SetVariables_#2}
  ]
  particleUpdateScript [
   &ParticleState()
   &ScaleSpriteSize(Uniform Curve Sprite Scale=DataInterface() {propertyValues {Curve {keys [
        {arriveTangent=1.2275, interpMode=RCIM_Cubic, leaveTangent=1.2275, tangentMode=RCTM_User, time=0, value=0.2}
        {arriveTangent=1.7266, interpMode=RCIM_Cubic, leaveTangent=1.7266, tangentMode=RCTM_User, time=1, value=1.4}
       ]}}})
   &DynamicMaterialParameters(
    Index 0 Param 1=&FloatFromCurve(
     FloatCurve=DataInterface() {propertyValues {Curve {keys [
         {arriveTangent=1.9451, interpMode=RCIM_Cubic, leaveTangent=1.9451, tangentMode=RCTM_User, time=0, value=0}
         {arriveTangent=0.6322, interpMode=RCIM_Cubic, leaveTangent=0.6322, tangentMode=RCTM_User, time=0.6, value=1}
        ]}}}
     Scale Curve=0.5
    )
    Index 0 Param 4=1
    Module.Param0WriteEnabled=true
    Module.Param1WriteEnabled=true
    Module.Param2WriteEnabled=false
    Module.Param3WriteEnabled=true
   )
   &ScaleColor(
    Scale Alpha=&Multiply_Float(
     A=&FloatFromCurve(FloatCurve=DataInterface() {propertyValues {Curve {keys [
          {interpMode=RCIM_Cubic, leaveTangent=5.1035, tangentMode=RCTM_Break, time=0, value=0}
          {arriveTangent=-0.2156, interpMode=RCIM_Cubic, leaveTangent=-0.2156, tangentMode=RCTM_User, time=0.3, value=0.8}
          {arriveTangent=-0.0478, interpMode=RCIM_Cubic, leaveTangent=-0.0478, tangentMode=RCTM_User, time=1, value=0}
         ]}}})
     B=&FNTimeOfDayScalar(Night Scale=1, Morning Scale=0.4, Day Scale=0.8, Evening Scale=0.3)
    )
    Scale RGB=&MakeVectorFromLinearColorRGB(Color=&ColorFromCurve(ColorCurve=DataInterface() {propertyValues {
        AlphaCurve {keys [<at 0.3955 1 RCIM_Cubic>]}
        BlueCurve {keys [<at 0 0.2111 RCIM_Cubic>]}
        GreenCurve {keys [<at 0 0.3777 RCIM_Cubic>]}
        RedCurve {keys [<at 0 0.5104 RCIM_Cubic>]}
       }}))
    ScaleA=true
    ScaleRGB=true
   ) {module=ScaleColor001}
   &SolveForcesAndVelocity(Clamp Velocity=false, Limit Acceleration=false)
  ]
  rendererClasses [&NiagaraSpriteRendererProperties]
  renderers [
   {rendererClass=&NiagaraSpriteRendererProperties}
   {rendererClass=&NiagaraSpriteRendererProperties, rendererIndex=1}
  ]
 }
 Smoke_Ground {
  emitterUpdateScript [
   &EmitterState()
   &SpawnBurst_Instantaneous(Spawn Count=12, Use Spawn Probability=false)
  ]
  particleSpawnScript [
   &InitializeParticle(
    Color=<rgba 1 1 1 0.1>
    Lifetime=&RandomRangeFloat(Minimum=0.5, Maximum=0.6)
    Sprite Rotation Mode=::NewEnumerator2
    Sprite Size Mode=::NewEnumerator1
    Uniform Sprite Size Max=35
    Uniform Sprite Size Min=25
    UsePositionOffset=false
    Write Material Random=true
   )
   &ShapeLocation(
    Cylinder Distribution=::NewEnumerator2
    Cylinder Height=0
    Cylinder Radius=1
    MaskSpawnGroup=false
    Offset=<0 0 -10>
    Offset Mode=::NewEnumerator0
    Rotation Coordinate Space=::World
    Shape Primitive=::NewEnumerator1
    U Position=&ReturnNormalizedExecIndex(Normalized Index Includes One=false)
    UseOwnerScale=true
   )
   &AddVelocity(Velocity=&Multiply_VectorByFloat(Vector=$Particles.ShapeLocation.ShapeNormal, Float=&RandomRangeFloat(Minimum=450, Maximum=550)))
   &ApplyInitialForces()
   &CameraOffset(Camera Offset Amount=40)
   /SetVariables(Particles.FirstFlipBookFrame:NiagaraInt32=&RandomRangeInt(Minimum=0, Maximum=64)) {isSetParameters=true, module=SetVariables_#1}
   /SetVariables(
    Particles.Stored Particle Size:Vector3f=$Particles.Scale
    Particles.Stored Sprite Size:Vector2f=&Multiply_Vector2D_ByFloat(Float=1, Vector2D=$Particles.SpriteSize)
   ) {isSetParameters=true, module=SetVariables_#2}
  ]
  particleUpdateScript [
   &ParticleState()
   /SetVariables(Particles.Scale=$StackContext.Stored Particle Size, Particles.SpriteSize=$StackContext.Stored Sprite Size) {isSetParameters=true}
   &AccelerationForce(Acceleration=&RandomRangeVector(Minimum=<0 0 -25>, Maximum=<1 1 -75>))
   &self:ViewpointBased_Sprite_Rotation_01(Rotation Rate=3) {module=ViewpointBased_Sprite_Rotation_01}
   &Drag(
    Drag=&Multiply_Float(
     A=&RandomRangeFloat(Minimum=15, Maximum=17)
     B=&FloatFromCurve(FloatCurve=DataInterface() {propertyValues {Curve {keys [
          {arriveTangent=-2.2549, interpMode=RCIM_Cubic, leaveTangent=-2.2549, tangentMode=RCTM_User, time=0, value=1}
          {interpMode=RCIM_Cubic, tangentMode=RCTM_User, time=1, value=0}
         ]}}})
    )
    Use Linear Drag=true
    Use Rotational Drag=true
   )
   &ScaleSpriteSize(Uniform Curve Sprite Scale=DataInterface() {propertyValues {Curve {keys [{arriveTangent=4.1286, interpMode=RCIM_Cubic, leaveTangent=4.1286, tangentMode=RCTM_User, time=0, value=0.2}, <at 1 1.4 RCIM_Cubic>]}}})
   &DynamicMaterialParameters(
    Index 0 Param 1=&FloatFromCurve(
     FloatCurve=DataInterface() {propertyValues {Curve {keys [
         {arriveTangent=1.9451, interpMode=RCIM_Cubic, leaveTangent=1.9451, tangentMode=RCTM_User, time=0, value=0}
         {arriveTangent=0.6322, interpMode=RCIM_Cubic, leaveTangent=0.6322, tangentMode=RCTM_User, time=0.6, value=1}
        ]}}}
     Scale Curve=0.5
    )
    Index 0 Param 4=1
    Module.Param0WriteEnabled=true
    Module.Param1WriteEnabled=true
    Module.Param2WriteEnabled=false
    Module.Param3WriteEnabled=true
   ) {module=DynamicMatParams}
   &ScaleColor(
    Scale Alpha=&Multiply_Float(
     A=&FloatFromCurve(FloatCurve=DataInterface() {propertyValues {Curve {keys [
          {interpMode=RCIM_Cubic, leaveTangent=5.1035, tangentMode=RCTM_Break, time=0, value=0}
          {arriveTangent=-0.2156, interpMode=RCIM_Cubic, leaveTangent=-0.2156, tangentMode=RCTM_User, time=0.3, value=0.8}
          {arriveTangent=-0.0478, interpMode=RCIM_Cubic, leaveTangent=-0.0478, tangentMode=RCTM_User, time=1, value=0}
         ]}}})
     B=&FNTimeOfDayScalar(Night Scale=1, Morning Scale=0.4, Day Scale=0.8, Evening Scale=0.3)
    )
    Scale RGB=&MakeVectorFromLinearColorRGB(Color=&ColorFromCurve(ColorCurve=DataInterface() {propertyValues {
        AlphaCurve {keys [<at 0.3955 1 RCIM_Cubic>]}
        BlueCurve {keys [<at 0 0.2206 RCIM_Cubic>]}
        GreenCurve {keys [<at 0 0.382 RCIM_Cubic>]}
        RedCurve {keys [<at 0 0.5104 RCIM_Cubic>]}
       }}))
    ScaleA=true
    ScaleRGB=true
   )
   &SolveForcesAndVelocity(Clamp Velocity=false, Limit Acceleration=false)
  ]
  rendererClasses [&NiagaraSpriteRendererProperties]
  renderers [
   {rendererClass=&NiagaraSpriteRendererProperties}
   {rendererClass=&NiagaraSpriteRendererProperties, rendererIndex=1}
  ]
 }
}
summary {
 systemName=NS_Emote_PizzaParty_DustPuff
 emitters [
  {emitterName=Smoke_Ground, rendererClasses [&NiagaraSpriteRendererProperties]}
  {emitterName=Directional_Smoke_Ground001, rendererClasses [&NiagaraSpriteRendererProperties]}
 ]
 userVariables [
  User.bIsFrontEnd {type=NiagaraBool}
  User.bIsFrontEndPreview {type=NiagaraBool}
  User.FootDustColorArray {
   type=NiagaraDataInterfaceArrayColor
   description=""
   defaultValue {
    struct=NiagaraExt_VariableValue_DataInterface
    value {
     dataInterface=&self:NiagaraDataInterfaceArrayColor_0
     dataInterfaceClass=&NiagaraDataInterfaceArrayColor
    }
   }
  }
  User.Global Scale {type=NiagaraFloat, defaultValue=0, description=""}
  User.SurfaceTypeIndex {type=NiagaraInt32}
 ]
}
